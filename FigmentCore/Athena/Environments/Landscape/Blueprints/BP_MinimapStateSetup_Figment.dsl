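; Blueprint that drives the "MiniMapState" scalar in the MPC_Landscape_S07 material parameter collection, controlling the Figment minimap's landscape state.
; When the match is ready, the authority sets State_Amount to 1 or 0 from a calendar event and takes progress from a stateful gameplay event.
; Remote clients write the raw value to the collection one second after its OnRep.
; The construction script and an editor command write a curve-remapped value instead.
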(event Custom|SetStateEd
  ;[1 1 1] Editor Commands
  (Rendering|Material|SetScalarParameterValue "/Game/Athena/Environments/Landscape/MPC/MPC_Landscape_S07.MPC_Landscape_S07" "MiniMapState" (Math|Curves|GetFloatValue (Variables|Default|GetCurve) (Variables|Default|GetState_Amount)) :!class /Script/Engine.KismetMaterialLibrary))

(event EventBeginPlay
  (Class|FortAsyncActionAuthorityMatchReady|MatchReadyAsync
    (:OnReady
      (CallFunction|RefreshCalender(AuthorityOnly))
      (bind (_asynctaskproxy _payload) (GameplayEventRouter|StartListeningToStatefulEvent (GameplayEventRouter|GetGlobalEventRouter) "/FigmentCore/Athena/Environments/Landscape/Blueprints/Figment_MinimapState_Structure.Figment_MinimapState_Structure")
        (:OnEventReceived
          _)
        (:OnSavedEventStateExists
          _))
      (Networking|SwitchHasAuthority
        (:Authority
          (bind _payload_progress_2_175a34fb44cef4b2b8eb8bb2e56a2cca (Utilities|Struct|BreakFigmentMinimapStateStructure _payload))
          (Variables|Default|SetState_Amount _payload_progress_2_175a34fb44cef4b2b8eb8bb2e56a2cca))))))

(event Custom|RefreshCalender%20%28Authority%20Only%29
  (Networking|SwitchHasAuthority
    (:Authority
      (bind _biseventactive (CalendarEvents|GetCalendarEventInformation (Variables|Default|GetCalendarEventName)))
      (if _biseventactive
        (Variables|Default|SetState_Amount 1.0)
        (else
          (Variables|Default|SetState_Amount 0.0))))))

(event Custom|SetMapState
  (params (NewParam :type Float%20%28double-precision%29))
  (Variables|Default|SetState_Amount NewParam))

(event Custom|SetMPCValue
  (Utilities|FlowControl|Delay 1.0)
  (Rendering|Material|SetScalarParameterValue "/Game/Athena/Environments/Landscape/MPC/MPC_Landscape_S07.MPC_Landscape_S07" "MiniMapState" (Variables|Default|GetState_Amount) :!class /Script/Engine.KismetMaterialLibrary))

(fn OnRep_State_Amount
  (Networking|SwitchHasAuthority
    (:Remote
      (Variables|Default|SetDidEverRepNotify true)
      (CallFunction|SetMPCValue))))

(fn ConstructionScript
  (Rendering|Material|SetScalarParameterValue "/Game/Athena/Environments/Landscape/MPC/MPC_Landscape_S07.MPC_Landscape_S07" "MiniMapState" (Math|Curves|GetFloatValue (Variables|Default|GetCurve) (Variables|Default|GetState_Amount)) :!class /Script/Engine.KismetMaterialLibrary))
